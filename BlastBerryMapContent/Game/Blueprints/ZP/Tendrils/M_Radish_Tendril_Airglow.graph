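# Unlit translucent Niagara ribbon material for the airglow around Radish tendrils.
# Panning noise blends two colors, optionally a rainbow texture, then particle color tints it for emissive adjustment.
# Opacity multiplies panning cloud noise, a cross-ribbon sine falloff, a dynamic-parameter length mask, Fresnel and particle alpha, with depth fade outside mobile feature levels.
# A gradient pulse branch is computed but does not feed the emissive.
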Material /BlastBerryMapContent/Game/Blueprints/ZP/Tendrils/M_Radish_Tendril_Airglow
graph {
 properties {bUsedWithNiagaraRibbons=true, blendMode=BLEND_Translucent, pixelDepthOffsetMode=PDOM_Legacy, shadingModel=MSM_Unlit, usedShadingModels=Unlit}
 nodes [
  n1 = TextureCoordinate()
  n2 = AppendVector(A=Constant() {r=1}, B=Constant() {r=1})
  n3 = LinearInterpolate(Alpha=DynamicParameter() {parameterIndex=1, defaultValue=<rgba 1 1 0 0>, paramNames [0, 1, Pulse_Offset, ColorNoiseOffset]}.Pulse_Offset) {constA=-0.6, constB=0.6}
  n4 = Constant()
  n5 = TextureSample(UVs=Add(A=Multiply(A=@n1, B=@n2), B=AppendVector(A=@n3, B=@n4))) {
   samplerSource=SSM_Clamp_WorldGroupSettings
   texture=/BlastBerryMapContent/Game/Blueprints/ZP/Tendrils/Textures/T_LinearGrad_Horizontal_Radish
  }
  n6 = Power(Base=@n5.R, Exp=ScalarParameter() {defaultValue=1, parameterName=Pulse_Power})
  n7 = ScalarParameter() {defaultValue=30, parameterName=Pulse_Mult}
  n8 = TextureCoordinate()
  n9 = AppendVector(A=ScalarParameter() {defaultValue=0.5, parameterName=ColorGrad_Tile_X}, B=Constant())
  n10 = ScalarParameter() {defaultValue=0.1, parameterName=ColorGrad_Pan_X}
  n11 = ScalarParameter() {defaultValue=0.01, parameterName=ColorGrad_Pan_Y}
  n12 = Add(
   A=AppendVector(
    A=DynamicParameter() {parameterIndex=1, defaultValue=<rgba 1 1 0 0>, paramNames [0, 1, Pulse_Offset, ColorNoiseOffset]}.ColorNoiseOffset
    B=Constant()
   )
   B=Panner(Coordinate=Multiply(A=@n8, B=@n9), Speed=AppendVector(A=@n10, B=@n11))
  )
  n13 = Multiply(
   A=Power(
    Base=TextureSample(UVs=@n12) {samplerType=SAMPLERTYPE_LinearColor, texture=&HighContrastNoise}.R
    Exp=ScalarParameter() {defaultValue=4, parameterName=ColorGrad_Power}
   )
   B=ScalarParameter() {defaultValue=2, parameterName=ColorGrad_Mult}
  )
  n14 = LinearInterpolate(
   A=LinearInterpolate(
    A=VectorParameter() {parameterName=ColorA, defaultValue=<rgba 0 0.3261 1 1>}.RGB
    B=VectorParameter() {parameterName=ColorB, defaultValue=<rgba 0.781 0 1 1>}.RGB
    Alpha=Clamp(None=@n13)
   )
   B=TextureSample(UVs=@n12) {texture=&/T_Radish_Tendril_Rainbow}.RGB
   Alpha=ScalarParameter() {parameterName=Use_Rainbow}
  )
  Add(A=Multiply(A=Multiply(A=@n6, B=@n7), B=@n14), B=@n14)
  n15 = DynamicParameter() {defaultValue=<rgba 1 1 0 0>, paramNames [Mask, Thread_Tiling_X, Noise_Tiling_X, ColorNoiseOffset_X]}
  n16 = Constant() {r=-0.5}
  n17 = Subtract(A=TextureCoordinate(), B=AppendVector(A=Add(A=@n15.Mask, B=@n16), B=Constant()))
  n18 = Clamp(None=Multiply(A=ComponentMask(None=@n17) {r=true}, B=Constant() {r=2}))
  Floor(None=@n18)
  MaterialFunctionCall() {materialFunction=&Fresnel_Function}
  MaterialFunctionCall() {materialFunction=&MF_Fresnel_FX}
  n19 = ParticleColor()
  MaterialFunctionCall(Emissive=Multiply(A=@n14, B=@n19.RGB)) {materialFunction=&MF_FX_EmissiveAdjustments}
  n20 = ComponentMask(None=TextureCoordinate()) {g=true}
  n21 = Clamp(None=Power(Base=Sine(None=@n20) {period=2}) {constExponent=2.2})
  n22 = OneMinus(None=@n18)
  n23 = TextureCoordinate() {coordinateIndex=1}
  n24 = AppendVector(A=Constant() {r=0.1}, B=Constant() {r=0.1})
  n25 = Constant() {r=0.05}
  n26 = Constant() {r=0.0125}
  n27 = TextureSample(UVs=Panner(Coordinate=Multiply(A=@n23, B=@n24), Speed=AppendVector(A=@n25, B=@n26))) {texture=&AtmosphericCloudNoise01}
  n28 = LinearInterpolate(A=Constant(), B=Constant() {r=1}, Alpha=Multiply(A=Power(Base=@n27.RGB) {constExponent=2.5}) {constB=2.5})
  n29 = Multiply(
   A=Multiply(A=Clamp(None=@n28), B=Multiply(A=@n21, B=@n22))
   B=MaterialFunctionCall(Brightness=Constant() {r=1}, Exponent=Constant() {r=10}) {materialFunction=&MF_Fresnel_FX}.Fresnel
  )
  n30 = DepthFade(Opacity=@n29, FadeDistance=ScalarParameter() {defaultValue=1000, parameterName=DepthFade_Distance})
  Multiply(A=@n19.A, B=FeatureLevelSwitch(Default=@n30, ES2=@n29, ES3_1=@n29, SM5=@n30, SM6=@n30))
 ]
}
